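# Material function that layers the gem effects for the Nebula Burden Cache back bling onto incoming material attributes.
# A shared masks function supplies center, side and accent gem masks, each multiplied by a scalar TechArtMask input.
# Center, Side and Accent sub-functions run in sequence, each passing its material to the next, and the Side stage also receives a left/right mask.
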MaterialFunction /BRCosmetics/Accessories/FORT_Backpacks/Backpack_M_MED_NebulaBurdenCache/MaterialFunctions/MF_NebulaBurdenCache_Master
graph {nodes [
  n1 = FunctionInput() {inputName=TechArtMask, inputType=FunctionInput_Scalar, sortPriority=1}
  n2 = MaterialFunctionCall() {materialFunction=&/MF_NebulaBurdenCache_Masks}
  n3 = MaterialFunctionCall!(
   Material=FunctionInput() {inputName=Material, inputType=FunctionInput_MaterialAttributes}
   Mask=Multiply(A=@n1, B=@n2.Center Gem)
   WarpMask
   Constellation Wisp Mask Opacity
  ) {materialFunction=&/MF_NebulaBurdenCache_Center}
  n4 = Multiply(A=@n1, B=@n2.Side Gems Combined)
  FunctionOutput(None=MaterialFunctionCall!(
    Material=MaterialFunctionCall!(Material=@n3.Material, Mask=@n4, Left Right Mask=@n2.Side Gems Left/Right Mask) {materialFunction=&/MF_NebulaBurdenCache_Side}.Material
    Mask=Multiply(A=@n2.Accent Gems, B=@n1)
   ) {materialFunction=&/MF_NebulaBurdenCache_Accent}.Material)
 ]}
