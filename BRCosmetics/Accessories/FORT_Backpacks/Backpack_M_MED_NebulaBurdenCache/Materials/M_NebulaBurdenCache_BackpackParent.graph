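# Parent material for the NebulaBurdenCache backpack cosmetic, set up for skeletal meshes and mesh deformers, with the shading model chosen by expression.
# The 2023 character PreFX function feeds the backpack's NebulaBurdenCache master function, whose result passes into the matching PostFX function.
# The master function's TechArtMask input is the saturated Tech Art Mask red channel minus itself, so it evaluates to zero.
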Material /BRCosmetics/Accessories/FORT_Backpacks/Backpack_M_MED_NebulaBurdenCache/Materials/M_NebulaBurdenCache_BackpackParent
graph {
 properties {
  bUseMaterialAttributes=true
  bUsedWithMeshDeformer=true
  bUsedWithSkeletalMesh=true
  pixelDepthOffsetMode=PDOM_Legacy
  shadingModel=MSM_FromMaterialExpression
  usedShadingModels=Default Lit | Subsurface | Clear Coat | Subsurface Profile | Cloth
 }
 nodes [
  n1 = MaterialFunctionCall() {materialFunction=&MF_CharacterParent_2023_PreFX}
  n2 = MaterialFunctionCall(Float3=@n1.Tech Art Mask) {materialFunction=&BreakOutFloat3Components}
  n3 = Subtract(A=@n2.R, B=@n2.R)
  MaterialFunctionCall(
   Material=MaterialFunctionCall!(Material=@n1.Result, TechArtMask=Saturate(None=@n3)) {materialFunction=/BRCosmetics/Accessories/FORT_Backpacks/Backpack_M_MED_NebulaBurdenCache/MaterialFunctions/MF_NebulaBurdenCache_Master}.Result
   M=@n1.M
   Pre Skinned Local Position=@n1.M
   Distance Blend=@n1.M
   Fresnel=@n1.M
  ) {materialFunction=&MF_CharacterParent_2023_PostFX}
 ]
}
