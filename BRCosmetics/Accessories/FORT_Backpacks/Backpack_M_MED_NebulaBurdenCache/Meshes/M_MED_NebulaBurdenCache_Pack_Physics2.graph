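# Physics asset for the NebulaBurdenCache backpack cosmetic, set up around its pack body and left and right handles.
# Five capsule bodies cover the pack, two small handle bases and two dynamic handle bones.
# Each base joins the pack with free angular motion.
# Each dynamic handle joins its base with zero twist and first-swing limits and a wider second-swing limit, so it flaps on one axis like a hinge.
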PhysicsAsset /BRCosmetics/Accessories/FORT_Backpacks/Backpack_M_MED_NebulaBurdenCache/Meshes/M_MED_NebulaBurdenCache_Pack_Physics2
bodies {
 base_handle_l [{shapeType=Capsule, length=0.505, radius=0.505, shapeName=base_handle_l_capsule, rotation=<rot 0 -90 0>}]
 base_handle_r [{shapeType=Capsule, length=0.505, radius=0.505, shapeName=base_handle_r_capsule, rotation=<rot 0 -90 0>}]
 dyn_handle_l [{shapeType=Capsule, length=1, radius=0.5, shapeName=dyn_handle_l_capsule, center=<2 0 0>, rotation=<rot 0 0.0708 -89.8286>}]
 dyn_handle_r [{shapeType=Capsule, length=1, radius=0.5, shapeName=dyn_handle_r_capsule, center=<2 0 0>, rotation=<rot 0 179.9292 -89.8286>}]
 pack [{shapeType=Capsule, length=6.5, radius=4.3153, shapeName=pack_capsule, center=<0 -9.6724 -0.1784>, rotation=<rot 0 89.9996 -89.9998>}]
}
constraints [
 {
  bone1Name=base_handle_l
  bone2Name=pack
  swing1LimitDegrees=45
  swing1Motion=Free
  swing2LimitDegrees=45
  swing2Motion=Free
  twistLimitDegrees=45
  twistMotion=Free
 }
 {bone1Name=dyn_handle_l, bone2Name=base_handle_l, swing1LimitDegrees=0, swing2LimitDegrees=65, twistLimitDegrees=0}
 {
  bone1Name=base_handle_r
  bone2Name=pack
  swing1LimitDegrees=45
  swing1Motion=Free
  swing2LimitDegrees=45
  swing2Motion=Free
  twistLimitDegrees=45
  twistMotion=Free
 }
 {bone1Name=dyn_handle_r, bone2Name=base_handle_r, swing1LimitDegrees=0, swing2LimitDegrees=65, twistLimitDegrees=0}
]
