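# Animation Blueprint for the NebulaBurdenCache backpack cosmetic, adding physics sway and body-aware placement.
# A RigidBody node simulates in root-bone space, fed by per-update gravity and sim-space settings.
# Falling selects its own settings; stationary jumps overwrite the Default settings variable instead of the output, otherwise Default applies.
# The pack bone then gets a small male-only offset, plus translation and roll mapped from the owner's spine bend.
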AnimBlueprint /BRCosmetics/Accessories/FORT_Backpacks/Backpack_M_MED_NebulaBurdenCache/Meshes/M_MED_NebulaBurdenCache_Pack_AnimBP
target_skeleton=&/M_MED_NebulaBurdenCache_Pack_Skeleton
graphs [
 AnimGraph {nodes [
   n1 = RigidBody!(
    ComponentPose=Local To Component(LocalPose=Input Pose())
    Alpha
    OverrideWorldGravity:Vector=Get Gravity Override()
    SimSpaceSettings:Sim Space Settings Structure=Get Sim Space Settings Output() -> (Sim Space Settings Output:Sim Space Settings Structure)
   ) -> (Pose) {
    bOverrideWorldGravity=true
    overridePhysicsAsset=&/M_MED_NebulaBurdenCache_Pack_Physics2
    simulationSpace=BaseBoneSpace
    alphaBoolBlend {blendOption=Linear}
    baseBoneRef {boneName=root}
    componentAppliedLinearAccClamp=<3000 3000 3000>
    componentLinearAccScale=<0.4 0.4 0.4>
    componentLinearVelScale=<0.4 0.4 0.4>
    overrideWorldGravity {z=-700}
    simSpaceSettings {maxAngularVelocity=3, maxLinearAcceleration=200, maxLinearVelocity=500, worldAlpha=0.2, externalLinearDragV=<0.1 0.1 0.1>}
   }
   n2 = "Transform (Modify) Bone - Bone: pack"!(ComponentPose=@n1, bAlphaBoolEnabled:bool=Get isMale() -> (isMale:bool), Translation="0,-0.15,0") -> (Pose) {
    alphaInputType=Bool
    translationMode=BMM_Additive
    translationSpace=BCS_BoneSpace
    alphaBoolBlend {blendOption=Linear}
    boneToModify {boneName=pack}
    translation {y=-0.15}
   }
   n3 = "Transform (Modify) Bone - Bone: pack"!(ComponentPose=@n2, Alpha:float, Translation=Get backpack_translation(), Rotation:Rotator=Get backpack_rotate()) -> (Pose) {rotationMode=BMM_Additive, translationMode=BMM_Additive, alphaBoolBlend {blendOption=Linear}, boneToModify {boneName=pack}}
   Output Pose(Result=Component To Local(ComponentPose=@n3)) {layerGroup=None}
  ]}
 EventGraph {nodes [
   Event OnCharacterCustomizationLoaded()
   > n1 = Set Owning Pawn(Owning Pawn:Pawn Object Reference=TryGetFortPawn()) -> (Output_Get:Pawn Object Reference)
   > Is Valid(InputObject=@n1)
   > n2 = FindGender!(execute, Object:Object Reference=Get Owning Pawn() -> (Owning Pawn:Pawn Object Reference)) -> (True:Exec, False:Exec, OutGender:EFortCustomGender Enum=Invalid)
   > Set isMale(isMale:bool="Equal (Enum)"(A:EFortCustomGender Enum=@n2, B:EFortCustomGender Enum=Male)) -> (Output_Get:bool)
   Event BlueprintUpdateAnimation()
   > n3 = Sequence!(execute) -> (then_0, then_1, then_2:Exec)
   n4 = Gravity Override State Function DEPRECATED(
    execute=@n3.then_1
    CharacterPartAnimInstance=Self-Reference()
    GameTimeInSeconds=GetGameTimeInSeconds()
    GravityOverrideParams=Get Gravity Override Params()
   )
   > Set Gravity Override(Gravity Override=@n4)
   n5 = Branch(execute=@n3.then_2, Condition=Get bIsFalling())
   > Set Sim Space Settings Output(Sim Space Settings Output:Sim Space Settings Structure=Get Sim Space Settings Falling() -> (Sim Space Settings Falling:Sim Space Settings Structure)) -> (Output_Get:Sim Space Settings Structure)
   n6 = OR Boolean!(A=Get bIsTacticalSprinting(), B=Get bIsSprinting(), C:bool=Get bIsMoving2D()) -> (ReturnValue)
   n7 = Branch(execute=@n5.else, Condition=AND Boolean(A=Get bIsJumping(), B=NOT Boolean(A=@n6)))
   > Set Sim Space Settings Default(Sim Space Settings Default:Sim Space Settings Structure=Get Sim Space Settings Solo Jump() -> (Sim Space Settings Solo Jump:Sim Space Settings Structure)) -> (Output_Get:Sim Space Settings Structure)
   Set Sim Space Settings Output(execute=@n7.else, Sim Space Settings Output=Get Sim Space Settings Default() -> (Sim Space Settings Default:Sim Space Settings Structure))
   n8 = GetOwnerSkeletalMeshComponent()
   n9 = BreakRotator(InRot=BreakTransform(InTransform=GetDeltaTransformFromRefPose(self=@n8, BoneName=spine_05, BaseName=spine_01)).Rotation)
   Set backpack_translation(execute=@n3.then_0, backpack_translation=MakeVector(Y=MapRangeClamped(Value=@n9.Yaw, InRangeA=0, InRangeB=-30, OutRangeB=-4)))
   > Set backpack_rotate(backpack_rotate=MakeRotator(Roll=MapRangeClamped(Value=@n9.Yaw, InRangeA=0, InRangeB=-30, OutRangeB=-15)))
   Event BlueprintInitializeAnimation()
   > Set Skel_Mesh(Skel_Mesh=GetOwnerSkeletalMeshComponent())
   TryGetPawnOwner()
  ]}
]
variables [
 backpack_rotate {type=Rotator}
 backpack_translation {type=Vector}
 Gravity Override {type=Vector}
 Gravity Override Params {type=GravityOverrideParamsStruct}
 isMale {type=boolean}
 Owning Pawn {type=Pawn, default=null}
 Sim Space Settings Default {
  type=SimSpaceSettings
  default {
   dampingAlpha=1
   maxAngularAcceleration=10000
   maxAngularVelocity=3
   maxLinearAcceleration=700
   maxLinearVelocity=500
   velocityScaleZ=1
   worldAlpha=0.4
   externalLinearDragV=<0.2 0.2 0.2>
  }
 }
 Sim Space Settings Falling {
  type=SimSpaceSettings
  default {
   dampingAlpha=1
   maxAngularAcceleration=10000
   maxAngularVelocity=3
   maxLinearAcceleration=50
   maxLinearVelocity=500
   velocityScaleZ=1
   worldAlpha=0.3
   externalLinearDragV=<0.1 0.1 0.1>
  }
 }
 Sim Space Settings Output {
  type=SimSpaceSettings
  default {
   dampingAlpha=1
   maxAngularAcceleration=10000
   maxAngularVelocity=3
   maxLinearAcceleration=700
   maxLinearVelocity=500
   velocityScaleZ=1
   worldAlpha=0.4
   externalLinearDragV=<0.2 0.2 0.2>
  }
 }
 Sim Space Settings Solo Jump {
  type=SimSpaceSettings
  default {
   dampingAlpha=1
   maxAngularAcceleration=10000
   maxAngularVelocity=3
   maxLinearAcceleration=1900
   maxLinearVelocity=500
   velocityScaleZ=1
   worldAlpha=0.5
   externalLinearDragV=<0.2 0.2 0.2>
  }
 }
 Skel_Mesh {type=SkeletalMeshComponent}
]
